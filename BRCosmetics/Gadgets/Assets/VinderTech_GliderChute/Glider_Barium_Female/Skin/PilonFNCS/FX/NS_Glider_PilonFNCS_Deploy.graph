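# Niagara deploy effect for the PilonFNCS glider skin, attached to bones or sockets on the glider's skeletal mesh.
# NS_Wheels bursts four spinning meshes that scale, fade and drive material parameters.
# Two emitters trail four tapering, fading ribbons each: FN_AttachedRibbon spawns only in gameplay, FN_AttachedRibbonFE only in front end, switched by FrontEnd modifiers zeroing spawn rate.
# Wheel and gameplay-ribbon colors multiply age gradients by a time-of-day tint.
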NiagaraSystem /BRCosmetics/Gadgets/Assets/VinderTech_GliderChute/Glider_Barium_Female/Skin/PilonFNCS/FX/NS_Glider_PilonFNCS_Deploy
dependencies {
 usedDataInterfaces [&NiagaraDataInterfaceCurve, &NiagaraDataInterfaceColorCurve]
 usedDynamicInputs [&SetBoolByFloatComparison, &Abs_Float]
 usedModules [
  &self:SystemSpawnScript.SystemScriptSource.SystemScriptGraph.NiagaraNodeAssignment_0.SetVariables_#1
  &self:SystemSpawnScript.SystemScriptSource.SystemScriptGraph.NiagaraNodeAssignment_1.SetVariables_#2
  &SystemState
  &FN_Integrator
 ]
}
emitters {
 FN_AttachedRibbon {
  emitterUpdateScript [
   &EmitterState(
    Inactive Response=::NewEnumerator1
    Life Cycle Mode=::NewEnumerator1
    Loop Behavior=::NewEnumerator1
    Loop Delay=&FN_FrontEndFloatModifier(Gameplay=1, Front End=0.7)
    Loop Duration=&FN_FrontEndFloatModifier(Gameplay=3, Front End=1.2)
    UseLoopDelay=true
   )
   &SpawnRate(
    SpawnRate=&FN_FrontEndFloatModifier(Gameplay=&FN_FrontEndFloatModifier(Gameplay=60, Front End=0), Front End=0)
    Use Spawn Probability=false
   )
  ]
  particleSpawnScript [
   &PartitionParticles(Number Of Partitions=4, Write Partition to Ribbon ID=::NewEnumerator0)
   &InitializeParticle(
    Lifetime=&FN_FrontEndFloatModifier(Gameplay=0.5, Front End=0.5)
    Ribbon Width=13
    Ribbon Width Mode=::NewEnumerator1
    Sprite Size Mode=::NewEnumerator3
    Uniform Sprite Size=13
    UsePositionOffset=false
    Write Material Random=true
   )
   &SkeletalMeshLocation(
    Bone / Socket Index=$Particles.PartitionParticles.Partition
    MaskSpawnGroup=false
    Mesh Sampling Type=::NewEnumerator5
    Orientation Sampling=::NewEnumerator1
    Position Sampling=::NewEnumerator2
    Skeletal Mesh=$System.SkeletalMesh
    Socket Sampling Mode=::NewEnumerator5
    Use Sampled Mesh Attribute as Ribbon ID=false
   )
  ]
  particleUpdateScript [
   &ParticleState()
   &SkeletalMeshLocation(
    Bone / Socket Index=$Particles.PartitionParticles.Partition
    MaskSpawnGroup=false
    Mesh Sampling Type=::NewEnumerator5
    Orientation Sampling=::NewEnumerator1
    Position Sampling=::NewEnumerator2
    Skeletal Mesh=$System.SkeletalMesh
    Socket Sampling Mode=::NewEnumerator5
    Use Sampled Mesh Attribute as Ribbon ID=false
   ) {module=SkeletalMeshLocation001}
   /SetVariables(Particles.Position=&Lerp_Vector(
     A=$Particles.SkeletalMeshLocation001.SampledPosition
     B=Linked() {linkedVariable {name=Particles.Position, type=NiagaraPosition}}
     Alpha=&FloatFromCurve(FloatCurve=DataInterface() {propertyValues {Curve {keys [<at 0 0>,<at 0.1 1>]}}})
    )) {isSetParameters=true}
   &InheritSourceMovement(
    Applied Position Scale=&VectorFromFloat(Value=&FN_FrontEndFloatModifier(Gameplay=0, Front End=1))
    ApplyForce=false
    ApplyPosition=true
    ApplyPositionFalloff=false
   )
   &SolveForcesAndVelocity(Clamp Velocity=false, Limit Acceleration=false)
   &ScaleColor(
    Scale Alpha=&FloatFromCurve(FloatCurve=DataInterface() {propertyValues {Curve {keys [{arriveTangent=-3.0471, interpMode=RCIM_Cubic, leaveTangent=-3.0471, tangentMode=RCTM_User, time=0, value=1}, <at 1 0 RCIM_Cubic>]}}})
    Scale RGB=&Multiply_Vector(
     A=&MakeVectorFromLinearColorRGB(Color=&ColorFromCurve(ColorCurve=DataInterface() {propertyValues {
         AlphaCurve {keys [<at 0 1 RCIM_Cubic>]}
         BlueCurve {keys [<at 0.0152 0>,<at 0.453 0.0515>,<at 1 0.0109>]}
         GreenCurve {keys [<at 0.0152 0.7615>,<at 0.453 0>,<at 1 0>]}
         RedCurve {keys [<at 0.0152 1>,<at 0.453 0.8177>,<at 1 0.1146>]}
        }}))
     B=&MakeVectorFromLinearColorRGB(Color=&ColorFromCurve(
       ColorCurve=DataInterface() {propertyValues {
         AlphaCurve {keys [<at 0 1>,<at 0.125 1>,<at 0.25 1>,<at 0.375 1>,<at 0.5 1>,<at 0.625 1>,<at 0.75 1>,<at 0.875 1>,<at 1 1>]}
         BlueCurve {keys [<at 0.125 0.8385>,<at 0.2152 0.6172>,<at 0.25 0.6172>,<at 0.291 0.6939>,<at 0.333 0.7379>,<at 0.375 0.7835>,<at 0.416 0.7991>,<at 0.458 0.8148>,<at 0.5 0.8308>,<at 0.541 0.8148>,<at 0.583 0.7991>,<at 0.625 0.7991>,<at 0.666 0.7529>,<at 0.708 0.7084>,<at 0.75 0.5711>,<at 0.791 0.4735>,<at 0.833 0.3231>,<at 0.8704 0.8385>]}
         GreenCurve {keys [<at 0.125 0.6268>,<at 0.2152 0.5333>,<at 0.25 0.6376>,<at 0.291 0.7529>,<at 0.333 0.7835>,<at 0.375 0.8148>,<at 0.416 0.7991>,<at 0.458 0.8148>,<at 0.5 0.8469>,<at 0.541 0.7991>,<at 0.583 0.7991>,<at 0.625 0.8148>,<at 0.666 0.7991>,<at 0.708 0.7529>,<at 0.75 0.5711>,<at 0.791 0.5029>,<at 0.833 0.5029>,<at 0.8704 0.6268>]}
         RedCurve {keys [<at 0.125 0.3712>,<at 0.2152 0.7084>,<at 0.25 0.8148>,<at 0.291 0.8879>,<at 0.333 0.8963>,<at 0.375 0.9131>,<at 0.416 0.8714>,<at 0.458 0.8714>,<at 0.5 0.9131>,<at 0.541 0.8632>,<at 0.583 0.8714>,<at 0.625 0.9131>,<at 0.666 0.8963>,<at 0.708 0.8714>,<at 0.75 0.7529>,<at 0.791 0.7529>,<at 0.833 0.8469>,<at 0.8704 0.3712>]}
        }}
       CurveIndex=Linked() {linkedVariable {name=NPC.FortniteNPC.TimeOfDay_0-1, type=NiagaraFloat}}
      ))
    )
    ScaleA=true
    ScaleRGB=true
   )
   /SetVariables(Particles.RibbonLinkOrder:NiagaraFloat=$Particles.NormalizedAge) {isSetParameters=true}
   /SetVariables(Particles.SecondRibbonCoor:LinearColor=&MultiplyLinearColor(
     Linear Color=<rgba 0.7292 0.7292 0.7292 0.5>
     Scale Factor=&MakeVector4_FromLinearColor(LinearColor=$Particles.Color)
    )) {isSetParameters=true}
   &ScaleRibbonWidth(Ribbon Width Scale=&FloatFromCurve(FloatCurve=DataInterface() {propertyValues {Curve {keys [
         {arriveTangent=-1.4049, interpMode=RCIM_Cubic, leaveTangent=-1.4049, tangentMode=RCTM_User, time=0, value=1}
         {arriveTangent=-0.3852, interpMode=RCIM_Cubic, leaveTangent=-0.3852, tangentMode=RCTM_User, time=1, value=0.3}
        ]}}}))
  ]
  rendererClasses [&NiagaraRibbonRendererProperties]
  renderers [
   {rendererClass=&NiagaraRibbonRendererProperties}
   {rendererClass=&NiagaraRibbonRendererProperties, rendererIndex=1}
  ]
 }
 FN_AttachedRibbonFE {
  emitterUpdateScript [
   &EmitterState(
    Life Cycle Mode=::NewEnumerator1
    Loop Behavior=::NewEnumerator1
    Loop Delay=&FN_FrontEndFloatModifier(Gameplay=0.7, Front End=0.7)
    Loop Duration=&FN_FrontEndFloatModifier(Gameplay=3, Front End=2)
    UseLoopDelay=true
   )
   &FN_SpawnRibbon_v2(Number of Ribbons=4, SpawnRate:NiagaraFloat=&FN_FrontEndFloatModifier(Gameplay=0, Front End=30))
  ]
  particleSpawnScript [
   /Game/Effects/Niagara/Modules/Ribbons/FN_InitializeRibbon_v4(
    Color:LinearColor=<rgba 1 0.5591 0 1>
    Enable Skeletal Position=true
    Interpolate Bone Sampling=false
    Lifetime:NiagaraFloat=&FN_FrontEndFloatModifier(Gameplay=0.7, Front End=0.5)
    Ribbon Width:NiagaraFloat=&FN_FrontEndFloatModifier(Gameplay=16, Front End=13)
    SkeletalMesh:NiagaraDataInterfaceSkeletalMesh=$System.SkeletalMesh
   )
   &FN_FlipRibbonV()
   &AddVelocity(Velocity=&FN_FrontEndVector3Modifier(Gameplay=<0 0 0>, Front End=<0 100 0>))
   /SetVariables() {isSetParameters=true}
  ]
  particleUpdateScript [
   &InheritSourceMovement(
    Applied Position Scale=&VectorFromFloat(Value=&FloatFromCurve(
      FloatCurve=DataInterface() {propertyValues {Curve {keys [<at 0.5 0.99>,<at 1 1>]}}}
      Scale Curve=&FN_FrontEndFloatModifier(Gameplay=1, Front End=1)
     ))
    ApplyForce=false
    ApplyPosition=true
    ApplyPositionFalloff=false
   )
   &SolveForcesAndVelocity(Clamp Velocity=false, Limit Acceleration=false)
   /Game/Effects/Niagara/Modules/Ribbons/FN_UpdateRibbon_v4(Detach Locked Particle=true, Enable Skeletal Position=true, SkeletalMesh:NiagaraDataInterfaceSkeletalMesh=$System.SkeletalMesh)
   &ScaleRibbonWidth(Ribbon Width Scale=&FloatFromCurve(FloatCurve=DataInterface() {propertyValues {Curve {keys [
         {arriveTangent=-3.2493, interpMode=RCIM_Cubic, leaveTangent=-3.2493, tangentMode=RCTM_User, time=0, value=1}
         {interpMode=RCIM_Cubic, tangentMode=RCTM_User, time=1, value=0}
        ]}}}))
   &ScaleColor(
    Scale Alpha=&FloatFromCurve(FloatCurve=DataInterface() {propertyValues {Curve {keys [
         {arriveTangent=-3.6049, interpMode=RCIM_Cubic, leaveTangent=-3.6049, tangentMode=RCTM_User, time=0, value=1}
         {interpMode=RCIM_Cubic, tangentMode=RCTM_User, time=1, value=0}
        ]}}})
    ScaleA=true
    ScaleRGB=true
   )
   &DynamicMaterialParameters(Module.Param0WriteEnabled=true, Module.Param1WriteEnabled=true, Module.Param2WriteEnabled=true, Module.Param3WriteEnabled=true)
   /SetVariables(Particles.SecondRibbonCoor:LinearColor=&MultiplyLinearColor(
     Linear Color=<rgba 0.15 0.086 0 0.5>
     Scale Factor=&MakeVector4_FromLinearColor(LinearColor=$Particles.Color)
    )) {isSetParameters=true}
  ]
  rendererClasses [&NiagaraRibbonRendererProperties]
  renderers [
   {rendererClass=&NiagaraRibbonRendererProperties}
   {rendererClass=&NiagaraRibbonRendererProperties, rendererIndex=1}
  ]
 }
 NS_Wheels {
  emitterUpdateScript [
   &EmitterState(
    Inactive Response=::NewEnumerator1
    Life Cycle Mode=::NewEnumerator1
    Loop Behavior=::NewEnumerator1
    Loop Delay=0.3
    Loop Duration=&FN_FrontEndFloatModifier(Gameplay=2, Front End=3)
    UseLoopDelay=true
   )
   &FN_Timer()
   &SpawnBurst_Instantaneous(Spawn Count=4, Use Spawn Probability=false, UseLoopCountLimit=false)
  ]
  particleSpawnScript [
   &FN_SetSpawnGroup()
   &InitializeParticle(
    Color=<rgba 1 1 1 5>
    Lifetime=3
    Mesh Scale=<1 0.3 0.3>
    Mesh Scale Mode=::NewEnumerator4
    UsePositionOffset=false
    Write Material Random=true
   )
  ]
  particleUpdateScript [
   &ParticleState()
   &ScaleMeshSize(Scale Factor=&VectorFromFloat(Value=&FloatFromCurve(FloatCurve=DataInterface() {propertyValues {Curve {keys [<at 0 0.5>,<at 0.5 1>,<at 0.64 1.8>,<at 0.74 1>]}}})))
   &SkeletalMeshLocation(
    Bone / Socket Index=$Particles.SpawnGroup
    MaskSpawnGroup=false
    Mesh Sampling Type=::NewEnumerator5
    Skeletal Mesh=$System.SkeletalMesh
    Socket Sampling Mode=::NewEnumerator5
   )
   /SetVariables(Particles.MeshOrientation=&MultiplyQuaternion(
     Quaternion A=&MultiplyQuaternion(
      Quaternion A=$Particles.SkeletalMeshLocation.SampledRotation
      Quaternion B=&MakeQuatFromAxisAngle(Axis=<1 0 0>, AngleInDegrees=&RandomRangeFloat(Minimum=-360, Maximum=360))
     )
     Quaternion B=&MakeQuatFromAxisAngle(Axis=<1 0 0>, AngleInDegrees=&Multiply_Float(A=360, B=$Engine.Time), NormalizeAxis=true)
    )) {isSetParameters=true}
   &DynamicMaterialParameters(
    Index 3 Param 1=&Add_Float(A=&RandomRangeFloat(Minimum=0, Maximum=1), B=$Emitter.FN_Timer.Value)
    Index 3 Param 2=&FloatFromCurve(FloatCurve=DataInterface() {propertyValues {Curve {keys [<at 0 0.1>,<at 0.4 0.15>,<at 0.615 0.25>,<at 0.8154 0.13>]}}})
    Module.Param0WriteEnabled3=true
    Module.Param1WriteEnabled3=true
    Module.Param2WriteEnabled3=true
    Module.Param3WriteEnabled3=true
    Write Parameter Index 0=false
    Write Parameter Index 3=true
   )
   &ScaleColor(
    Scale Alpha=&FloatFromCurve(FloatCurve=DataInterface() {propertyValues {Curve {keys [<at 0 0>,<at 0.1 1>,<at 0.6 1>,<at 0.7 0>]}}})
    Scale RGB=&Multiply_Vector(
     A=&MakeVectorFromLinearColorRGB(Color=&ColorFromCurve(ColorCurve=DataInterface() {propertyValues {
         AlphaCurve {keys [<at 0 1 RCIM_Cubic>]}
         BlueCurve {keys [
           <at 0 0.0137 RCIM_Cubic>
           <at 0.2962 0.0408>
           {arriveTangent=-0.1421, interpMode=RCIM_Cubic, leaveTangent=-0.1421, time=0.5015, value=0}
           <at 0.5836 0>
           <at 0.6813 0>
          ]}
         GreenCurve {keys [
           <at 0 0 RCIM_Cubic>
           <at 0.2962 0>
           {arriveTangent=1.3118, interpMode=RCIM_Cubic, leaveTangent=1.3118, time=0.5015, value=0.238}
           <at 0.5836 0.377>
           <at 0.6813 0.1652>
          ]}
         RedCurve {keys [
           <at 0 0.2656 RCIM_Cubic>
           <at 0.2962 0.7917>
           {arriveTangent=0.7249, interpMode=RCIM_Cubic, leaveTangent=0.7249, time=0.5015, value=1}
           <at 0.5836 1>
           <at 0.6813 1>
          ]}
        }}))
     B=&MakeVectorFromLinearColorRGB(Color=&ColorFromCurve(
       ColorCurve=DataInterface() {propertyValues {
         AlphaCurve {keys [<at 0 1>,<at 0.125 1>,<at 0.25 1>,<at 0.375 1>,<at 0.5 1>,<at 0.625 1>,<at 0.75 1>,<at 0.875 1>,<at 1 1>]}
         BlueCurve {keys [<at 0.125 0.8385>,<at 0.2152 0.6172>,<at 0.25 0.6172>,<at 0.291 0.6939>,<at 0.333 0.7379>,<at 0.375 0.7835>,<at 0.416 0.7991>,<at 0.458 0.8148>,<at 0.5 0.8308>,<at 0.541 0.8148>,<at 0.583 0.7991>,<at 0.625 0.7991>,<at 0.666 0.7529>,<at 0.708 0.7084>,<at 0.75 0.5711>,<at 0.791 0.4735>,<at 0.833 0.3231>,<at 0.8704 0.8385>]}
         GreenCurve {keys [<at 0.125 0.6268>,<at 0.2152 0.5333>,<at 0.25 0.6376>,<at 0.291 0.7529>,<at 0.333 0.7835>,<at 0.375 0.8148>,<at 0.416 0.7991>,<at 0.458 0.8148>,<at 0.5 0.8469>,<at 0.541 0.7991>,<at 0.583 0.7991>,<at 0.625 0.8148>,<at 0.666 0.7991>,<at 0.708 0.7529>,<at 0.75 0.5711>,<at 0.791 0.5029>,<at 0.833 0.5029>,<at 0.8704 0.6268>]}
         RedCurve {keys [<at 0.125 0.3712>,<at 0.2152 0.7084>,<at 0.25 0.8148>,<at 0.291 0.8879>,<at 0.333 0.8963>,<at 0.375 0.9131>,<at 0.416 0.8714>,<at 0.458 0.8714>,<at 0.5 0.9131>,<at 0.541 0.8632>,<at 0.583 0.8714>,<at 0.625 0.9131>,<at 0.666 0.8963>,<at 0.708 0.8714>,<at 0.75 0.7529>,<at 0.791 0.7529>,<at 0.833 0.8469>,<at 0.8704 0.3712>]}
        }}
       CurveIndex=Linked() {linkedVariable {name=NPC.FortniteNPC.TimeOfDay_0-1, type=NiagaraFloat}}
      ))
    )
    ScaleA=true
    ScaleRGB=true
   )
  ]
  rendererClasses [&NiagaraMeshRendererProperties]
  renderers [{rendererClass=&NiagaraMeshRendererProperties, rendererIndex=0}]
 }
}
summary {
 systemName=NS_Glider_PilonFNCS_Deploy
 emitters [
  {emitterName=NS_Wheels, rendererClasses [&NiagaraMeshRendererProperties]}
  {emitterName=FN_AttachedRibbon, rendererClasses [&NiagaraRibbonRendererProperties]}
  {emitterName=FN_AttachedRibbonFE, rendererClasses [&NiagaraRibbonRendererProperties]}
 ]
 userVariables [
  User.bIsFrontEnd {type=NiagaraBool}
  User.bIsFrontEndPreview {type=NiagaraBool}
  User.bIsFullyDeployed {type=NiagaraBool}
  User.ForwardDot {type=NiagaraFloat, defaultValue=0, description=""}
  User.RibbonIDOffset {type=NiagaraInt32}
  User.RightDot {type=NiagaraFloat, defaultValue=0, description=""}
  User.SkeletalMesh {
   type=NiagaraDataInterfaceSkeletalMesh
   defaultValue {
    struct=NiagaraExt_VariableValue_DataInterface
    value {dataInterface=&self:NiagaraDataInterfaceSkeletalMesh_0}
   }
  }
 ]
}
